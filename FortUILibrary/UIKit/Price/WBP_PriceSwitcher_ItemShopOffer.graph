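# Item shop offer variant of WBP_PriceSwitcher_Base that displays an offer's price and supports a manual MTX price override.
# While constructed, it monitors the MTX store view model for updates if it has a real price.
# A positive override clears the offer view model, hides the strikethrough and resets currency, switcher state and price texts to MTX.
# Helper functions derive real-price text with the override applied and collapse empty strikethrough prices.
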WidgetBlueprint /FortUILibrary/UIKit/Price/WBP_PriceSwitcher_ItemShopOffer
graphs [
 EventGraph {nodes [
   Event Construct()
   > Branch(Condition=Get HasRealPrice() -> (HasRealPrice:bool))
   n1 = Get FortMtxStoreVM() -> (FortMtxStoreVM:Fort Mtx Store VM Object Reference)
   > StartMonitoringForUpdates!(execute, self:Fort Mcp Catalog VM Object Reference=@n1, RequestingObject:Object Reference) -> (then)
   Event Destruct()
   > Branch(Condition=Get HasRealPrice())
   > StopMonitoringForUpdates!(execute, self:Fort Mcp Catalog VM Object Reference=@n1, RequestingObject:Object Reference) -> (then)
  ]}
 GetRealPriceTextWithOverrides {nodes [
   n1 = GetRealPriceTextWithOverrides!() -> (
    then
    FortMtxStoreVM:Fort Mtx Store VM Object Reference
    LoadingState:EFortMtxStoreLoadingState Enum=Uninitialized
    MtxPrice:int
    HasRealPrice:bool=~
    Currency:EFortPriceCurrency Enum=Mtx
    IsPriceShowing:bool=~
    IsAdditionalText:bool=~
   ) {is=title K2Node_FunctionEntry}
   n2 = Get OverrideMtxPrice() -> (OverrideMtxPrice:int=0)
   n3 = Get MtxPrice() -> (MtxPrice:int=0)
   n4 = "integer > integer"(A=@n2)
   n5 = Get Currency() -> (Currency:EFortPriceCurrency Enum=Mtx)
   > Return Node!(
    execute
    Text:Text=GetRealPriceText!(
     FortMTXStoreVM:Fort Mtx Store VM Object Reference=@n1
     LoadingState:EFortMtxStoreLoadingState Enum=@n1.LoadingState
     MtxPrice:int=Select(Option 0:int=@n3, Option 1:int=@n2, Index=@n4) -> (ReturnValue:int)
     HasRealPrice:bool=@n1.HasRealPrice
     Currency:EFortPriceCurrency Enum=Select(Option 0:EFortPriceCurrency Enum=@n5, Option 1:EFortPriceCurrency Enum=Mtx, Index=@n4) -> (ReturnValue:EFortPriceCurrency Enum)
     IsPriceShowing:bool=@n1.IsPriceShowing
     IsAdditionalText:bool=@n1.IsAdditionalText
    ) -> (Text:Text)
   )
  ]}
 GetStrikethroughPriceVisibilityFromText {nodes [
   n1 = GetStrikethroughPriceVisibilityFromText!() -> (then, RegularPrice:Text) {is=title K2Node_FunctionEntry}
   > Return Node!(
    execute
    Visibility:ESlateVisibility Enum=Select(Option 0:ESlateVisibility Enum=SelfHitTestInvisible, Option 1:ESlateVisibility Enum=Collapsed, Index=TextIsEmpty(InText=@n1)) -> (ReturnValue:ESlateVisibility Enum)
   )
  ]}
 SetOverrideMtxPrice {nodes [
   n1 = SetOverrideMtxPrice!() -> (then, OfferMtxPrice:int) {is=title K2Node_FunctionEntry}
   > n2 = Set OverrideMtxPrice(OverrideMtxPrice:int=@n1) -> (Output_Get:int=0)
   > Branch(Condition="integer > integer"(A=@n2))
   > Set FortItemShopOfferVM(FortItemShopOfferVM:Fort Item Shop Offer VM Object Reference) -> (Output_Get:Fort Item Shop Offer VM Object Reference)
   > SetStrikethroughPriceVisibility!(execute, self:WBP Price Switcher Base Object Reference, Visibility:ESlateVisibility Enum=Collapsed) -> (then)
   > SetCurrency!(execute, self:WBP Price Switcher Base Object Reference, Currency:EFortPriceCurrency Enum=Mtx) -> (then)
   > SetSwitcherState!(
    execute
    self:WBP Price Switcher Base Object Reference
    PriceState:E_PriceSwitcherState Enum=GetPriceInfoState!(self:WBP Price Switcher Base Object Reference, IsOwned:bool=false, OfferType:EFortItemShopOfferType Enum=Regular) -> (SwitcherState:E_PriceSwitcherState Enum=NewEnumerator0)
   ) -> (then)
   > SetPrice!(execute, self:WBP Price Switcher Base Object Reference, Price:Text="To Text (Integer)"(Value=Get OfferMtxPrice() -> (OfferMtxPrice:int=0))) -> (then)
   > SetAdditionalPriceText!(execute, self:WBP Price Switcher Base Object Reference, Text:Text) -> (then)
   n3 = Get FortMtxStoreVM()
   > SetRealPriceText!(
    execute
    self:WBP Price Switcher Base Object Reference
    Text:Text=GetRealPriceTextWithOverrides!(
     self
     FortMtxStoreVM:Fort Mtx Store VM Object Reference=@n3
     LoadingState:EFortMtxStoreLoadingState Enum=Get LoadingState!(self:Fort Mtx Store VM Object Reference=@n3) -> (LoadingState:EFortMtxStoreLoadingState Enum=Uninitialized)
     MtxPrice:int=Get OverrideMtxPrice() -> (OverrideMtxPrice=0)
     HasRealPrice:bool=Get HasRealPrice()
     Currency:EFortPriceCurrency Enum=Mtx
     IsPriceShowing:bool=Get IsPriceShowing() -> (IsPriceShowing:bool)
     IsAdditionalText:bool=Get IsAdditionalRealPrice() -> (IsAdditionalRealPrice:bool)
    ) -> (Text:Text) {is=title K2Node_CallFunction}
   ) -> (then)
  ]}
]
variables [OverrideMtxPrice {type=integer}]
widgets {
 info {
  inheritedWidgetCount=0
  parentClass=/FortUILibrary/Systems/ItemShopTile/WBP_PriceSwitcher_Base.WBP_PriceSwitcher_Base_C
  rootWidgetClass=null
  widgetCount=0
 }
 widgets []
}
