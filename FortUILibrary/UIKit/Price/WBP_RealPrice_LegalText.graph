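# Price widget that shows real-money legal disclaimer text for a V-Bucks price, using a single common text block.
# PreConstruct applies exposed font, color and wrap settings, previews sample text at design time and collapses it at runtime.
# Construct and Destruct start and stop store view model monitoring.
# GetLegalText requests text from that view model only for a ready, shown, positive V-Bucks price, otherwise returning empty; SetLegalText writes text to the block.
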WidgetBlueprint /FortUILibrary/UIKit/Price/WBP_RealPrice_LegalText
graphs [
 EventGraph {nodes [
   n1 = Event PreConstruct()
   > SetFont(self=Get Text!() -> (Text:Common Text Object Reference), InFontInfo=Get TextFont() -> (TextFont:Slate Font Info Structure))
   > Set ColorAndOpacity!(execute, ColorAndOpacity:Slate Color Structure=Get TextColor() -> (TextColor:Slate Color Structure), self:Text Object Reference=Get Text()) -> (then, Output_Get:Slate Color Structure)
   > Set WrapTextAt!(execute, WrapTextAt:float=Get WrapTextAt() -> (WrapTextAt:double), self:Text Layout Widget Object Reference=Get Text()) -> (then, Output_Get:float)
   > Set AutoWrapText!(execute, AutoWrapText:bool=Get AutoWrap() -> (AutoWrap:bool), self:Text Layout Widget Object Reference=Get Text()) -> (then, Output_Get:bool)
   > n2 = Branch(Condition=@n1.IsDesignTime)
   > Set Text(Text=Get PreviewText() -> (PreviewText:Text), self=Get Text())
   SetVisibility(execute=@n2.else, self=Get Text(), InVisibility=Collapsed)
   Event Destruct()
   n3 = Get FortMtxStoreVM() -> (FortMtxStoreVM:Fort Mtx Store VM Object Reference)
   > StopMonitoringForUpdates!(execute, self:Fort Mcp Catalog VM Object Reference=@n3, RequestingObject:Object Reference) -> (then)
   Event Construct()
   > StartMonitoringForUpdates!(execute, self:Fort Mcp Catalog VM Object Reference=@n3, RequestingObject:Object Reference) -> (then)
  ]}
 GetLegalText {nodes [
   n1 = GetLegalText!() -> (then, LoadingState:EFortMtxStoreLoadingState Enum=Uninitialized, Price:int, Currency:EFortPriceCurrency Enum=Mtx, IsPriceShowing:bool=~) {is=title K2Node_FunctionEntry}
   > n2 = Branch(Condition=AND Boolean!(
     A="Equal (Enum)"(A:EFortMtxStoreLoadingState Enum=@n1, B:EFortMtxStoreLoadingState Enum=Ready)
     B="integer > integer"(A=@n1.Price)
     C:bool="Equal (Enum)"(A:EFortPriceCurrency Enum=@n1.Currency, B:EFortPriceCurrency Enum=Mtx)
     D:bool=Get IsPriceShowing() -> (IsPriceShowing:bool)
    ) -> (ReturnValue))
   > Return Node!(
    execute
    Text:Text=GetRealPriceLegalText!(
     self:Fort Mtx Store VM Object Reference=Get FortMtxStoreVM()
     MtxPrice:int=Get Price() -> (Price:int=0)
     bIsAdditionalText:bool=Get IsAdditionalText() -> (IsAdditionalText:bool)
    ) -> (ReturnValue:Text)
   )
   Return Node(execute=@n2.else)
  ]}
 SetLegalText {nodes [
   n1 = SetLegalText!() -> (then, Text:Text) {is=title K2Node_FunctionEntry}
   > Set Text(Text=@n1, self=Get Text())
  ]}
]
variables [
 AutoWrap {type=boolean}
 Currency {type=EFortPriceCurrency, default=Mtx}
 IsAdditionalText {type=boolean}
 IsPriceShowing {type=boolean, default=true}
 PreviewText {type=string, default="*To buy exactly 1,000 V-Bucks, you’d pay €8.99."}
 Price {type=integer}
 TextColor {type=SlateColor, default {colorUseRule=UseColor_Specified, specifiedColor=<rgba 1 1 1 0.65>}}
 TextFont {
  type=SlateFontInfo
  default {
   bForceMonospaced=false
   bMaterialIsStencil=false
   fontMaterial=null
   fontObject=&HeadingNow
   letterSpacing=0
   monospacedWidth=1
   size=11.25
   skewAmount=0
   typefaceFontName=74 Regular
   outlineSettings {
    bApplyOutlineToDropShadows=false
    bMiteredCorners=false
    bSeparateFillAlpha=false
    outlineMaterial=null
    outlineSize=0
    outlineColor=<rgba 0 0 0 1>
   }
  }
 }
 WrapTextAt {type=number, default=600}
]
widgets {
 info {inheritedWidgetCount=0, parentClass=&UserWidget, rootWidgetClass=&CommonTextBlock, widgetCount=1}
 widgets [Text {
   widgetClassPath=&CommonTextBlock
   bIsVariable=true
   widget=Text
   properties {
    bAutoCollapseWithEmptyText=true
    mobileFontSizeMultiplier=1.4
    text="[ph] *To buy exactly 1,000 V-Bucks, you’d pay €8.99."
    font {
     bForceMonospaced=false
     bMaterialIsStencil=false
     fontMaterial=null
     fontObject=&HeadingNow
     letterSpacing=0
     monospacedWidth=1
     size=11.25
     skewAmount=0
     typefaceFontName=74 Regular
     outlineSettings {
      bApplyOutlineToDropShadows=false
      bMiteredCorners=false
      bSeparateFillAlpha=false
      outlineMaterial=null
      outlineSize=0
      outlineColor=<rgba 0 0 0 1>
     }
    }
    shadowOffset=<0 0>
   }
  }]
}
